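; Dynamic UI director component that shows a vehicle dashboard scene while the local player drives a vehicle with matching tags.
; Once the UI is ready, it binds to the player controller's driver-enter, seat-change and exit events.
; Driver entry caches whether the vehicle has any configured tags, and seat changes and exits reuse that flag.
; AddOrRemoveDashboard adds the scene only when requested and the pawn is driving, otherwise removing it.
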(event EventOnReadyToShowUI
  (bind _asfort_player_controller_athena (Utilities|Casting|CastToFortPlayerControllerAthena (CallFunction|GetOwningLocalPlayerController)))
  (Vehicle|BindEventtoOnVehicleSeatChanged _asfort_player_controller_athena (EventDispatchers|CreateEvent :!event Custom|OnVehicleSeatChanged))
  (Vehicle|AssignOnEnterVehicleDriver _asfort_player_controller_athena (EventDispatchers|CreateEvent :!event Custom|OnEnterVehicle))
  (Vehicle|BindEventtoOnExitVehicle _asfort_player_controller_athena (EventDispatchers|CreateEvent :!event Custom|OnExitVehicle)))

(event Custom|OnEnterVehicle
  (bind _asfort_player_controller_athena (Utilities|Casting|CastToFortPlayerControllerAthena (CallFunction|GetOwningLocalPlayerController)))
  (bind _asfort_player_pawn (Utilities|Casting|CastToFortPlayerPawn (Pawn|GetControlledPawn _asfort_player_controller_athena)))
  (bind _output_get (Variables|Default|SetbVehicleTagsMatch (GameplayTags|HasAnyTags (Class|FortVehicleInterface|GetVehicleTags (Vehicle|GetVehicle _asfort_player_pawn)) (Variables|Default|GetVehicle%20Tags))))
  (if _output_get
    (Default|AddorRemoveDashboard true)))

(event Custom|OnVehicleSeatChanged
  (params (SeatIndex :type Integer))
  (if (Variables|Default|GetbVehicleTagsMatch)
    (Default|AddorRemoveDashboard true)))

(event Custom|OnExitVehicle
  (if (Variables|Default|GetbVehicleTagsMatch)
    (Default|AddorRemoveDashboard false)))

(fn AddOrRemoveDashboard
  (params (bAdd :type Boolean))
  (bind _returnvalue (CallFunction|GetOwningFortDynamicUIDirector))
  (bind _dashboard_scene (Variables|Default|GetDashboard%20Scene))
  (bind _isindriverseat (Default|IsinDriverSeat))
  (if (and (Variables|GetbAdd) _isindriverseat)
    (DynamicUIDirector|AddScene _returnvalue _dashboard_scene)
    (else
      (DynamicUIDirector|RemoveScene _returnvalue _dashboard_scene))))

(fn IsInDriverSeat
  (bind _asfort_player_pawn (Utilities|Casting|CastToFortPlayerPawn (Pawn|GetControlledPawn (DynamicUIDirector|GetOwningLocalPlayerController (CallFunction|GetOwningFortDynamicUIDirector))))
    (:then
      (return (Vehicle|IsDrivingVehicle _asfort_player_pawn)))
    (:CastFailed
      (return false))))
